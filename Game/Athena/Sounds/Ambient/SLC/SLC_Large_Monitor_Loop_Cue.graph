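# Ambient indoor SoundCue, named as a large monitor loop for the SLC set, but its root is null and it holds no nodes, so it has no audio source.
# Only playback settings remain: the small SLC ambient attenuation, the indoor Athena ambient sound class and a slight volume boost.
# A concurrency override caps it at one instance, with timed voice-steal release and volume-scale fades.
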SoundCue /Game/Athena/Sounds/Ambient/SLC/SLC_Large_Monitor_Loop_Cue
root=null
nodes []
properties {
 attenuationSettings=/Game/Athena/Sounds/Ambient/SLC/Att/SLC_Ambient_Small_Att
 bOverrideConcurrency=true
 maxDistance=1600
 soundClassObject=&Fort_Ambient_Indoors_Athena
 volumeMultiplier=1.253
 concurrencyOverrides {bVolumeScaleCanRelease=true, maxCount=1, voiceStealReleaseTime=2, volumeScaleAttackTime=1, volumeScaleReleaseTime=1}
}
