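# Looping ambient sound cue of randomized monitor beeps for the SLC ambient set, routed to the SFX sound class.
# A loop repeatedly plays one of seven equally weighted beep waves, each after a random delay of up to three seconds, with a modulator varying pitch and volume.
# A small-ambient attenuation asset handles spatial falloff, and a concurrency override limits playback to a single instance.
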SoundCue /Game/Athena/Sounds/Ambient/SLC/SLC_Monitor_Beep_Loop_Cue
root=Looping_0
nodes [
 {id=Delay_0, children [Random_1], properties {delayMax=3}}
 {id=Looping_0, children [Modulator_0]}
 {id=Modulator_0, children [Delay_0], properties {pitchMax=1.2, pitchMin=0.8, volumeMax=1.3, volumeMin=1.2}}
 {
  id=Random_1
  children [WavePlayer_4, WavePlayer_5, WavePlayer_6, WavePlayer_7, WavePlayer_8, WavePlayer_9, WavePlayer_11]
  properties {weights [1,1,1,1,1,1,1]}
 }
 {id=WavePlayer_11, properties {soundWaveAssetPtr=/Game/Athena/Sounds/Ambient/SLC/Wavs/SLC_Monitor_Beep_08}}
 {id=WavePlayer_4, properties {soundWaveAssetPtr=/Game/Athena/Sounds/Ambient/SLC/Wavs/SLC_Monitor_Beep_01}}
 {id=WavePlayer_5, properties {soundWaveAssetPtr=/Game/Athena/Sounds/Ambient/SLC/Wavs/SLC_Monitor_Beep_02}}
 {id=WavePlayer_6, properties {soundWaveAssetPtr=/Game/Athena/Sounds/Ambient/SLC/Wavs/SLC_Monitor_Beep_03}}
 {id=WavePlayer_7, properties {soundWaveAssetPtr=/Game/Athena/Sounds/Ambient/SLC/Wavs/SLC_Monitor_Beep_04}}
 {id=WavePlayer_8, properties {soundWaveAssetPtr=/Game/Athena/Sounds/Ambient/SLC/Wavs/SLC_Monitor_Beep_05}}
 {id=WavePlayer_9, properties {soundWaveAssetPtr=/Game/Athena/Sounds/Ambient/SLC/Wavs/SLC_Monitor_Beep_06}}
]
properties {
 attenuationSettings=/Game/Athena/Sounds/Ambient/SLC/Att/SLC_Ambient_Small_Att
 bOverrideConcurrency=true
 maxDistance=2200
 soundClassObject=&SFX
 volumeMultiplier=0.5
 concurrencyOverrides {bVolumeScaleCanRelease=true, maxCount=1, voiceStealReleaseTime=2, volumeScaleAttackTime=1, volumeScaleReleaseTime=1}
}
