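# Ambient printer sound cue for the SLC sound set, layering a steady hum with intermittent movement sounds.
# A looping branch repeatedly picks one of four printer-move waves after a random delay, pitched about an octave up with varied volume.
# A Mixer blends it with a continuous printer loop pitched an octave down.
# It uses small-ambient attenuation, the SFX class, and a single-voice concurrency override.
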SoundCue /Game/Athena/Sounds/Ambient/SLC/SLC_Printer_04_Ambient_Loop_Cue
root=Mixer_0
nodes [
 {id=Delay_0, children [Random_0], properties {delayMax=3}}
 {id=Looping_0, children [Modulator_0]}
 {id=Mixer_0, children [Looping_0,Modulator_1], properties {inputVolume [1.6,1.6]}}
 {id=Modulator_0, children [Delay_0], properties {pitchMax=2.1, pitchMin=2, volumeMax=0.8, volumeMin=0.5}}
 {id=Modulator_1, children [WavePlayer_10], properties {pitchMax=0.5, pitchMin=0.5, volumeMax=0.7, volumeMin=0.5}}
 {id=Random_0, children [WavePlayer_0,WavePlayer_1,WavePlayer_2,WavePlayer_3], properties {weights [1,1,1,1]}}
 {id=WavePlayer_0, properties {soundWaveAssetPtr=/Game/Athena/Sounds/Ambient/SLC/Wavs/SLC_Printer_Move_01}}
 {id=WavePlayer_1, properties {soundWaveAssetPtr=/Game/Athena/Sounds/Ambient/SLC/Wavs/SLC_Printer_Move_02}}
 {id=WavePlayer_10, properties {bLooping=true, soundWaveAssetPtr=/Game/Athena/Sounds/Ambient/SLC/Wavs/SLC_Printer_Loop}}
 {id=WavePlayer_2, properties {soundWaveAssetPtr=/Game/Athena/Sounds/Ambient/SLC/Wavs/SLC_Printer_Move_03}}
 {id=WavePlayer_3, properties {soundWaveAssetPtr=/Game/Athena/Sounds/Ambient/SLC/Wavs/SLC_Printer_Move_04}}
]
properties {
 attenuationSettings=/Game/Athena/Sounds/Ambient/SLC/Att/SLC_Ambient_Small_Att
 bOverrideConcurrency=true
 maxDistance=2200
 soundClassObject=&SFX
 volumeMultiplier=3
 concurrencyOverrides {bVolumeScaleCanRelease=true, maxCount=1, voiceStealReleaseTime=2, volumeScaleAttackTime=1, volumeScaleReleaseTime=3}
}
